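# Niagara system for the DryEraseToro backpack's glow lights, tracing its glass, door, light strips and arrow.
# Burst-spawned sprites interpolate between socket-reader positions: glass and door circle four-socket loops, strips run twelve edges between eight sockets, the arrow steps along two lines.
# A shared glow color fades in and out over each loop, with socket-rotated direction fed to materials.
# Only glass and door add light renderers; CapsoleGlow001 is disabled.
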NiagaraSystem /BRCosmetics/Accessories/FORT_Backpacks/Backpack_M_MED_DryEraseToro/FX/NS_Backpack_DryEraseToro_GlowLights
dependencies {
 usedDataInterfaces [&NiagaraDataInterfaceCurve, &NiagaraDataInterfaceSocketReader]
 usedDynamicInputs []
 usedModules [
  &self:SystemSpawnScript.SystemScriptSource.SystemScriptGraph.NiagaraNodeAssignment_1.SetVariables_#1
  &self:SystemSpawnScript.SystemScriptSource.SystemScriptGraph.NiagaraNodeAssignment_2.SetVariables_#2
  &SystemState
 ]
}
emitters {
 Arrow_Glow {
  emitterSpawnScript [/SetVariables(Emitter.CapsuleGlowDirection:Vector3f=<1 0 0>) {isSetParameters=true, module=SetVariables_#3}]
  emitterUpdateScript [
   &EmitterState(Inactive Response=::NewEnumerator1, Life Cycle Mode=::NewEnumerator1, Loop Behavior=::NewEnumerator1, Loop Duration=8.7, UseLoopDelay=false)
   &SpawnBurst_Instantaneous(Spawn Count=8, Use Spawn Probability=false, UseLoopCountLimit=false)
  ]
  particleSpawnScript [
   &InitializeParticle(
    Color=$System.Glow Color
    Lifetime=2
    Sprite Size Mode=::NewEnumerator3
    Uniform Sprite Size=0.4
    UsePositionOffset=false
    Write Material Random=true
   ) {module=InitializeParticle001}
   &CameraOffset()
   &PartitionParticles(Normalized Index In Partition Includes One=true, Particles Per Partition=4, Partition Mode:ENiagara_PartitionMode_v2=::NewEnumerator1)
   /SetVariables(Particles.SpriteSizeCapsule:Vector2f=<1.5 1.5>) {isSetParameters=true, module=SetVariables_#4}
  ]
  particleUpdateScript [
   &ParticleState(Kill Particles When Lifetime Has Elapsed=false, Loop Particles Lifetime=true)
   &SocketLocation(
    Position Sampling=::NewEnumerator1
    SamplePosOffset=false
    Socket Index=16
    Socket Reader=$System.Socket Reader
    Socket Sampling Mode=::NewEnumerator3
    VelocitySpeedLimit=false
   )
   &SocketLocation(
    Position Sampling=::NewEnumerator1
    SamplePosOffset=false
    Socket Index=17
    Socket Reader=$System.Socket Reader
    Socket Sampling Mode=::NewEnumerator3
    VelocitySpeedLimit=false
   ) {module=SocketLocation001}
   &SocketLocation(
    Position Sampling=::NewEnumerator1
    SamplePosOffset=false
    Socket Index=18
    Socket Reader=$System.Socket Reader
    Socket Sampling Mode=::NewEnumerator3
    VelocitySpeedLimit=false
   ) {module=SocketLocation002}
   &ScaleColor(
    Scale Alpha=&FloatFromCurve(FloatCurve=DataInterface() {propertyValues {Curve {keys [
         {arriveTangent=4.1785, interpMode=RCIM_Cubic, leaveTangent=4.1785, tangentMode=RCTM_User, time=0, value=0}
         {arriveTangent=1.5009, interpMode=RCIM_Cubic, leaveTangent=1.5009, tangentMode=RCTM_User, time=0.2, value=0.55}
         {arriveTangent=0.0224, interpMode=RCIM_Cubic, leaveTangent=0.0224, tangentMode=RCTM_User, time=1, value=1}
        ]}}})
    Scale RGB=&VectorFromFloat(Value=&FloatFromCurve(
      FloatCurve=DataInterface() {propertyValues {Curve {
         defaultValue=0
         keys [
          <at 0 0 RCIM_Cubic>
          {arriveTangent=0.0159, interpMode=RCIM_Cubic, leaveTangent=0.0159, tangentMode=RCTM_User, time=0.088, value=1}
          {interpMode=RCIM_Cubic, tangentMode=RCTM_User, time=0.9, value=1}
          {interpMode=RCIM_Cubic, tangentMode=RCTM_User, time=1, value=0}
         ]
        }}}
      CurveIndex=$Emitter.NormalizedLoopAge
      Scale Curve=3
     ))
    ScaleA=false
    ScaleRGB=true
   )
   /SetVariables(Particles.Position=&MergeHelper_VectorSwitch(
     0=&Lerp_Vector(
      A=Linked() {linkedVariable {name=Output.SocketLocation.SampledPosition, type=NiagaraPosition}}
      B=Linked() {linkedVariable {name=Output.SocketLocation001.SampledPosition, type=NiagaraPosition}}
      Alpha=$Particles.Initial.PartitionParticles.Normalized Index In Partition
     )
     1=&Lerp_Vector(
      A=Linked() {linkedVariable {name=Output.SocketLocation.SampledPosition, type=NiagaraPosition}}
      B=Linked() {linkedVariable {name=Output.SocketLocation002.SampledPosition, type=NiagaraPosition}}
      Alpha=$Particles.Initial.PartitionParticles.Normalized Index In Partition
     )
     2=?
     3=?
     Index=$Particles.PartitionParticles.Partition
     Num=::NewEnumerator0
    )) {isSetParameters=true, module=SetVariables_#5}
   /SetVariables(Particles.Position=&Add_Vector(
     A=Linked() {linkedVariable {name=Particles.Position, type=NiagaraPosition}}
     B=&Multiply_VectorByFloat(
      Vector=<0 0 -1.12>
      Float=/Niagara/DynamicInputs/Integer/V2/Round_Float(Float:NiagaraFloat=&ModuloFloat(A=&Multiply_Float(A=$Particles.NormalizedAge, B=4), B=1))
     )
    )) {isSetParameters=true}
   &FN_Light_Attributes(Create Exponent Attribute=true, Create Radius Attribute=true, Exponent=5, Radius=8)
   /SetVariables(
    Particles.Capsule Color:LinearColor=&MultiplyLinearColor(
     Linear Color=$Particles.Color
     Scale Factor=&MakeVector4FromVectorAndFloat("Vector (RGB)"=&VectorFromFloat(Value=1), "Float (Alpha)"=0.15)
    )
    Particles.LightColor=&MultiplyLinearColor(Linear Color=$Particles.Color, Scale Factor=&MakeVector4_FromFloat(Float=2))
   ) {isSetParameters=true, module=SetVariables_#6}
   &DynamicMaterialParameters(
    Index 0 Param 1=&MakeFloatFromVector(
     Vector=&RotateVectorByQuaternion(Vector To Rotate=$Emitter.CapsuleGlowDirection, Quaternion=$Output.SocketLocation.SampledRotation)
     Channel=::NewEnumerator0
    )
    Index 0 Param 2=&MakeFloatFromVector(
     Vector=&RotateVectorByQuaternion(Vector To Rotate=$Emitter.CapsuleGlowDirection, Quaternion=$Output.SocketLocation.SampledRotation)
     Channel=::NewEnumerator1
    )
    Index 0 Param 3=&MakeFloatFromVector(
     Vector=&RotateVectorByQuaternion(Vector To Rotate=$Emitter.CapsuleGlowDirection, Quaternion=$Output.SocketLocation.SampledRotation)
     Channel=::NewEnumerator2
    )
    Index 0 Param 4=1.5
    Module.Param0WriteEnabled=true
    Module.Param1WriteEnabled=true
    Module.Param2WriteEnabled=true
    Module.Param3WriteEnabled=true
   )
  ]
  rendererClasses [&NiagaraSpriteRendererProperties]
  renderers [
   {rendererClass=&NiagaraSpriteRendererProperties}
   {rendererClass=&NiagaraSpriteRendererProperties, rendererIndex=1}
  ]
 }
 CapsoleGlow001 {
  bEnabled=false
  emitterSpawnScript [/SetVariables(Emitter.CapsuleGlowDirection:Vector3f=<1 0 0>) {isSetParameters=true}]
  emitterUpdateScript [
   &EmitterState(Inactive Response=::NewEnumerator1, Life Cycle Mode=::NewEnumerator1, Loop Behavior=::NewEnumerator1, Loop Duration=8, UseLoopDelay=false)
   &SpawnBurst_Instantaneous(Spawn Count=21, Use Spawn Probability=false, UseLoopCountLimit=false)
   &SpawnRate(Use Spawn Probability=false) {enabled=false}
  ]
  particleSpawnScript [
   &InitializeParticle(
    Color=$System.Glow Color
    Lifetime=2
    Sprite Size Mode=::NewEnumerator3
    Uniform Sprite Size=3
    UsePositionOffset=false
    Write Material Random=true
   )
   &CameraOffset(Camera Offset Amount=3)
   &PartitionParticles(
    Normalized Index In Partition Includes One=true
    Number Of Partitions=3
    Particles Per Partition=7
    Partition Mode:ENiagara_PartitionMode_v2=::NewEnumerator1
   )
  ]
  particleUpdateScript [
   &ParticleState(Kill Particles When Lifetime Has Elapsed=false, Loop Particles Lifetime=true)
   &DynamicMaterialParameters(
    Index 0 Param 1=&MakeFloatFromVector(
     Vector=&RotateVectorByQuaternion(Vector To Rotate=$Emitter.CapsuleGlowDirection, Quaternion=$Output.SocketLocation.SampledRotation)
     Channel=::NewEnumerator0
    )
    Index 0 Param 2=&MakeFloatFromVector(
     Vector=&RotateVectorByQuaternion(Vector To Rotate=$Emitter.CapsuleGlowDirection, Quaternion=$Output.SocketLocation.SampledRotation)
     Channel=::NewEnumerator1
    )
    Index 0 Param 3=&MakeFloatFromVector(
     Vector=&RotateVectorByQuaternion(Vector To Rotate=$Emitter.CapsuleGlowDirection, Quaternion=$Output.SocketLocation.SampledRotation)
     Channel=::NewEnumerator2
    )
    Index 0 Param 4=3
    Module.Param0WriteEnabled=true
    Module.Param1WriteEnabled=true
    Module.Param2WriteEnabled=true
    Module.Param3WriteEnabled=true
   )
   &ScaleColor(
    Scale Alpha=0.25
    Scale RGB=&VectorFromFloat(Value=&FloatFromCurve(
      FloatCurve=DataInterface() {propertyValues {Curve {
         defaultValue=0
         keys [
          <at 0 0 RCIM_Cubic>
          {arriveTangent=0.0159, interpMode=RCIM_Cubic, leaveTangent=0.0159, tangentMode=RCTM_User, time=0.1, value=1}
          {interpMode=RCIM_Cubic, tangentMode=RCTM_User, time=0.9, value=1}
          {interpMode=RCIM_Cubic, tangentMode=RCTM_User, time=1, value=0}
         ]
        }}}
      CurveIndex=$Emitter.NormalizedLoopAge
     ))
    ScaleA=true
    ScaleRGB=true
   ) {module=ScaleColor001}
   &SocketLocation(
    Position Sampling=::NewEnumerator1
    SamplePosOffset=false
    Socket Reader=$System.Socket Reader
    Socket Sampling Mode=::NewEnumerator3
    VelocitySpeedLimit=false
   )
   &SocketLocation(
    Position Sampling=::NewEnumerator1
    SamplePosOffset=false
    Socket Index=1
    Socket Reader=$System.Socket Reader
    Socket Sampling Mode=::NewEnumerator3
    VelocitySpeedLimit=false
   ) {module=SocketLocation001}
   &SocketLocation(
    Position Sampling=::NewEnumerator1
    SamplePosOffset=false
    Socket Index=2
    Socket Reader=$System.Socket Reader
    Socket Sampling Mode=::NewEnumerator3
    VelocitySpeedLimit=false
   ) {module=SocketLocation002}
   &SocketLocation(
    Position Sampling=::NewEnumerator1
    SamplePosOffset=false
    Socket Index=3
    Socket Reader=$System.Socket Reader
    Socket Sampling Mode=::NewEnumerator3
    VelocitySpeedLimit=false
   ) {module=SocketLocation003}
   /SetVariables(Particles.OffsetNormalizedAge:NiagaraFloat=&ModuloFloat(
     A=&Add_Float(
      A=$Particles.NormalizedAge
      B=&Add_Float(
       A=&Multiply_Float(A=$Particles.PartitionParticles.NormalizedPartition, B=1)
       B=&Multiply_Float(A=$Particles.PartitionParticles.Normalized Index In Partition, B=0.1)
      )
     )
     B=1
    )) {isSetParameters=true}
   /SetVariables(Particles.Position=&MergeHelper_VectorSwitch(
     0=&LerpPosition(
      A=$Output.SocketLocation.SampledPosition
      B=$Output.SocketLocation001.SampledPosition
      Alpha=&ClampFloat(Float=&Multiply_Float(A=$Particles.OffsetNormalizedAge, B=4))
     )
     1=&LerpPosition(
      A=$Output.SocketLocation001.SampledPosition
      B=$Output.SocketLocation002.SampledPosition
      Alpha=&ClampFloat(Float=&Subtract_Float(A=&Multiply_Float(A=$Particles.OffsetNormalizedAge, B=4), B=1))
     )
     2=&LerpPosition(
      A=$Output.SocketLocation002.SampledPosition
      B=$Output.SocketLocation003.SampledPosition
      Alpha=&ClampFloat(Float=&Subtract_Float(A=&Multiply_Float(A=$Particles.OffsetNormalizedAge, B=4), B=2))
     )
     3=&LerpPosition(
      A=$Output.SocketLocation003.SampledPosition
      B=$Output.SocketLocation.SampledPosition
      Alpha=&ClampFloat(Float=&Subtract_Float(A=&Multiply_Float(A=$Particles.OffsetNormalizedAge, B=4), B=3))
     )
     Index=&MakeCustomIntFromBool(
      Bool=&SetBoolByFloatComparison(A=$Particles.OffsetNormalizedAge, B=0.25, Comparison Type=::NewEnumerator6)
      True INT=0
      False INT=&MakeCustomIntFromBool(
       Bool=&SetBoolByFloatComparison(A=$Particles.OffsetNormalizedAge, B=0.5, Comparison Type=::NewEnumerator6)
       True INT=1
       False INT=&MakeCustomIntFromBool(
        Bool=&SetBoolByFloatComparison(A=$Particles.OffsetNormalizedAge, B=0.75, Comparison Type=::NewEnumerator6)
        True INT=2
        False INT=3
       )
      )
     )
     Num=::NewEnumerator2
    )) {isSetParameters=true}
  ]
  rendererClasses [&NiagaraSpriteRendererProperties]
  renderers [{rendererClass=&NiagaraSpriteRendererProperties}]
 }
 Door_Glow {
  emitterSpawnScript [/SetVariables(Emitter.CapsuleGlowDirection:Vector3f=<1 0 0>) {isSetParameters=true, module=SetVariables_#3}]
  emitterUpdateScript [
   &EmitterState(Inactive Response=::NewEnumerator1, Life Cycle Mode=::NewEnumerator1, Loop Behavior=::NewEnumerator1, Loop Duration=8.7, UseLoopDelay=false)
   &SpawnBurst_Instantaneous(Spawn Count=21, Use Spawn Probability=false, UseLoopCountLimit=false)
  ]
  particleSpawnScript [
   &InitializeParticle(
    Color=$System.Glow Color
    Lifetime=2
    Sprite Size Mode=::NewEnumerator3
    Uniform Sprite Size=0.4
    UsePositionOffset=false
    Write Material Random=true
   ) {module=InitializeParticle001}
   &CameraOffset()
   &PartitionParticles(
    Normalized Index In Partition Includes One=true
    Number Of Partitions=3
    Particles Per Partition=7
    Partition Mode:ENiagara_PartitionMode_v2=::NewEnumerator1
   )
   /SetVariables(Particles.SpriteSizeCapsule:Vector2f=<1.5 1.5>) {isSetParameters=true, module=SetVariables_#4}
  ]
  particleUpdateScript [
   &ParticleState(Kill Particles When Lifetime Has Elapsed=false, Loop Particles Lifetime=true)
   &SocketLocation(
    Position Sampling=::NewEnumerator1
    SamplePosOffset=false
    Socket Index=12
    Socket Reader=$System.Socket Reader
    Socket Sampling Mode=::NewEnumerator3
    VelocitySpeedLimit=false
   )
   &SocketLocation(
    Position Sampling=::NewEnumerator1
    SamplePosOffset=false
    Socket Index=13
    Socket Reader=$System.Socket Reader
    Socket Sampling Mode=::NewEnumerator3
    VelocitySpeedLimit=false
   ) {module=SocketLocation001}
   &SocketLocation(
    Position Sampling=::NewEnumerator1
    SamplePosOffset=false
    Socket Index=14
    Socket Reader=$System.Socket Reader
    Socket Sampling Mode=::NewEnumerator3
    VelocitySpeedLimit=false
   ) {module=SocketLocation002}
   &SocketLocation(
    Position Sampling=::NewEnumerator1
    SamplePosOffset=false
    Socket Index=15
    Socket Reader=$System.Socket Reader
    Socket Sampling Mode=::NewEnumerator3
    VelocitySpeedLimit=false
   ) {module=SocketLocation003}
   &ScaleColor(
    Scale Alpha=&FloatFromCurve(FloatCurve=DataInterface() {propertyValues {Curve {keys [
         {arriveTangent=4.1785, interpMode=RCIM_Cubic, leaveTangent=4.1785, tangentMode=RCTM_User, time=0, value=0}
         {arriveTangent=1.5009, interpMode=RCIM_Cubic, leaveTangent=1.5009, tangentMode=RCTM_User, time=0.2, value=0.55}
         {arriveTangent=0.0224, interpMode=RCIM_Cubic, leaveTangent=0.0224, tangentMode=RCTM_User, time=1, value=1}
        ]}}})
    Scale RGB=&VectorFromFloat(Value=&FloatFromCurve(
      FloatCurve=DataInterface() {propertyValues {Curve {
         defaultValue=0
         keys [
          <at 0 0 RCIM_Cubic>
          {arriveTangent=0.0159, interpMode=RCIM_Cubic, leaveTangent=0.0159, tangentMode=RCTM_User, time=0.088, value=1}
          {interpMode=RCIM_Cubic, tangentMode=RCTM_User, time=0.9, value=1}
          {interpMode=RCIM_Cubic, tangentMode=RCTM_User, time=1, value=0}
         ]
        }}}
      CurveIndex=$Emitter.NormalizedLoopAge
      Scale Curve=3
     ))
    ScaleA=false
    ScaleRGB=true
   )
   /SetVariables(Particles.OffsetNormalizedAge:NiagaraFloat=&ModuloFloat(
     A=&Add_Float(
      A=$Particles.NormalizedAge
      B=&Add_Float(
       A=&Multiply_Float(A=$Particles.PartitionParticles.NormalizedPartition, B=1)
       B=&Multiply_Float(A=$Particles.PartitionParticles.Normalized Index In Partition, B=0.1)
      )
     )
     B=1
    )) {isSetParameters=true, module=SetVariables_#7}
   /SetVariables(Particles.Position=&MergeHelper_VectorSwitch(
     0=&LerpPosition(
      A=$Output.SocketLocation.SampledPosition
      B=$Output.SocketLocation001.SampledPosition
      Alpha=&ClampFloat(Float=&Multiply_Float(A=$Particles.OffsetNormalizedAge, B=4))
     )
     1=&LerpPosition(
      A=$Output.SocketLocation001.SampledPosition
      B=$Output.SocketLocation002.SampledPosition
      Alpha=&ClampFloat(Float=&Subtract_Float(A=&Multiply_Float(A=$Particles.OffsetNormalizedAge, B=4), B=1))
     )
     2=&LerpPosition(
      A=$Output.SocketLocation002.SampledPosition
      B=$Output.SocketLocation003.SampledPosition
      Alpha=&ClampFloat(Float=&Subtract_Float(A=&Multiply_Float(A=$Particles.OffsetNormalizedAge, B=4), B=2))
     )
     3=&LerpPosition(
      A=$Output.SocketLocation003.SampledPosition
      B=$Output.SocketLocation.SampledPosition
      Alpha=&ClampFloat(Float=&Subtract_Float(A=&Multiply_Float(A=$Particles.OffsetNormalizedAge, B=4), B=3))
     )
     Index=&MakeCustomIntFromBool(
      Bool=&SetBoolByFloatComparison(A=$Particles.OffsetNormalizedAge, B=0.25, Comparison Type=::NewEnumerator6)
      True INT=0
      False INT=&MakeCustomIntFromBool(
       Bool=&SetBoolByFloatComparison(A=$Particles.OffsetNormalizedAge, B=0.5, Comparison Type=::NewEnumerator6)
       True INT=1
       False INT=&MakeCustomIntFromBool(
        Bool=&SetBoolByFloatComparison(A=$Particles.OffsetNormalizedAge, B=0.75, Comparison Type=::NewEnumerator6)
        True INT=2
        False INT=3
       )
      )
     )
     Num=::NewEnumerator2
    )) {isSetParameters=true, module=SetVariables_#5}
   &FN_Light_Attributes(
    Create Exponent Attribute=true
    Create Radius Attribute=true
    Enable Particle Light Enable=true
    Exponent=5
    Light Enabled:NiagaraBool=&SetBoolByIntComparison(A=$Particles.PartitionParticles.Index In Partition, B=3, Comparison Type=::NewEnumerator4)
    Radius=5
   )
   /SetVariables(
    Particles.Capsule Color:LinearColor=&MultiplyLinearColor(
     Linear Color=$Particles.Color
     Scale Factor=&MakeVector4FromVectorAndFloat("Vector (RGB)"=&VectorFromFloat(Value=1), "Float (Alpha)"=0.15)
    )
    Particles.LightColor=&MultiplyLinearColor(Linear Color=$Particles.Color, Scale Factor=&MakeVector4_FromFloat(Float=1))
   ) {isSetParameters=true, module=SetVariables_#6}
   &DynamicMaterialParameters(
    Index 0 Param 1=&MakeFloatFromVector(
     Vector=&RotateVectorByQuaternion(Vector To Rotate=$Emitter.CapsuleGlowDirection, Quaternion=$Output.SocketLocation.SampledRotation)
     Channel=::NewEnumerator0
    )
    Index 0 Param 2=&MakeFloatFromVector(
     Vector=&RotateVectorByQuaternion(Vector To Rotate=$Emitter.CapsuleGlowDirection, Quaternion=$Output.SocketLocation.SampledRotation)
     Channel=::NewEnumerator1
    )
    Index 0 Param 3=&MakeFloatFromVector(
     Vector=&RotateVectorByQuaternion(Vector To Rotate=$Emitter.CapsuleGlowDirection, Quaternion=$Output.SocketLocation.SampledRotation)
     Channel=::NewEnumerator2
    )
    Index 0 Param 4=1.5
    Module.Param0WriteEnabled=true
    Module.Param1WriteEnabled=true
    Module.Param2WriteEnabled=true
    Module.Param3WriteEnabled=true
   )
  ]
  rendererClasses [&NiagaraLightRendererProperties, &NiagaraSpriteRendererProperties]
  renderers [
   {rendererClass=&NiagaraLightRendererProperties, rendererIndex=0}
   {rendererClass=&NiagaraSpriteRendererProperties, rendererIndex=1}
   {rendererClass=&NiagaraSpriteRendererProperties, rendererIndex=2}
  ]
 }
 Glass_Glow {
  emitterSpawnScript [/SetVariables(Emitter.CapsuleGlowDirection:Vector3f=<1 0 0>) {isSetParameters=true, module=SetVariables_#3}]
  emitterUpdateScript [
   &EmitterState(Inactive Response=::NewEnumerator1, Life Cycle Mode=::NewEnumerator1, Loop Behavior=::NewEnumerator1, Loop Duration=8.7, UseLoopDelay=false)
   &SpawnBurst_Instantaneous(Spawn Count=21, Use Spawn Probability=false, UseLoopCountLimit=false)
  ]
  particleSpawnScript [
   &InitializeParticle(
    Color=$System.Glow Color
    Lifetime=2
    Sprite Size Mode=::NewEnumerator3
    Uniform Sprite Size=0.4
    UsePositionOffset=false
    Write Material Random=true
   ) {module=InitializeParticle001}
   &CameraOffset()
   &PartitionParticles(
    Normalized Index In Partition Includes One=true
    Number Of Partitions=3
    Particles Per Partition=7
    Partition Mode:ENiagara_PartitionMode_v2=::NewEnumerator1
   )
   /SetVariables(Particles.SpriteSizeCapsule:Vector2f=<1.5 1.5>) {isSetParameters=true, module=SetVariables_#4}
  ]
  particleUpdateScript [
   &ParticleState(Kill Particles When Lifetime Has Elapsed=false, Loop Particles Lifetime=true)
   &SocketLocation(
    Position Sampling=::NewEnumerator1
    SamplePosOffset=false
    Socket Reader=$System.Socket Reader
    Socket Sampling Mode=::NewEnumerator3
    VelocitySpeedLimit=false
   )
   &SocketLocation(
    Position Sampling=::NewEnumerator1
    SamplePosOffset=false
    Socket Index=1
    Socket Reader=$System.Socket Reader
    Socket Sampling Mode=::NewEnumerator3
    VelocitySpeedLimit=false
   ) {module=SocketLocation001}
   &SocketLocation(
    Position Sampling=::NewEnumerator1
    SamplePosOffset=false
    Socket Index=2
    Socket Reader=$System.Socket Reader
    Socket Sampling Mode=::NewEnumerator3
    VelocitySpeedLimit=false
   ) {module=SocketLocation002}
   &SocketLocation(
    Position Sampling=::NewEnumerator1
    SamplePosOffset=false
    Socket Index=3
    Socket Reader=$System.Socket Reader
    Socket Sampling Mode=::NewEnumerator3
    VelocitySpeedLimit=false
   ) {module=SocketLocation003}
   &ScaleColor(
    Scale Alpha=&FloatFromCurve(FloatCurve=DataInterface() {propertyValues {Curve {keys [
         {arriveTangent=4.1785, interpMode=RCIM_Cubic, leaveTangent=4.1785, tangentMode=RCTM_User, time=0, value=0}
         {arriveTangent=1.5009, interpMode=RCIM_Cubic, leaveTangent=1.5009, tangentMode=RCTM_User, time=0.2, value=0.55}
         {arriveTangent=0.0224, interpMode=RCIM_Cubic, leaveTangent=0.0224, tangentMode=RCTM_User, time=1, value=1}
        ]}}})
    Scale RGB=&VectorFromFloat(Value=&FloatFromCurve(
      FloatCurve=DataInterface() {propertyValues {Curve {
         defaultValue=0
         keys [
          <at 0 0 RCIM_Cubic>
          {arriveTangent=0.0159, interpMode=RCIM_Cubic, leaveTangent=0.0159, tangentMode=RCTM_User, time=0.088, value=1}
          {interpMode=RCIM_Cubic, tangentMode=RCTM_User, time=0.9, value=1}
          {interpMode=RCIM_Cubic, tangentMode=RCTM_User, time=1, value=0}
         ]
        }}}
      CurveIndex=$Emitter.NormalizedLoopAge
      Scale Curve=3
     ))
    ScaleA=false
    ScaleRGB=true
   )
   /SetVariables(Particles.OffsetNormalizedAge:NiagaraFloat=&ModuloFloat(
     A=&Add_Float(
      A=$Particles.NormalizedAge
      B=&Add_Float(
       A=&Multiply_Float(A=$Particles.PartitionParticles.NormalizedPartition, B=1)
       B=&Multiply_Float(A=$Particles.PartitionParticles.Normalized Index In Partition, B=0.05)
      )
     )
     B=1
    )) {isSetParameters=true, module=SetVariables_#7}
   /SetVariables(Particles.Position=&MergeHelper_VectorSwitch(
     0=&LerpPosition(
      A=$Output.SocketLocation.SampledPosition
      B=$Output.SocketLocation001.SampledPosition
      Alpha=&ClampFloat(Float=&Multiply_Float(A=$Particles.OffsetNormalizedAge, B=4))
     )
     1=&LerpPosition(
      A=$Output.SocketLocation001.SampledPosition
      B=$Output.SocketLocation002.SampledPosition
      Alpha=&ClampFloat(Float=&Subtract_Float(A=&Multiply_Float(A=$Particles.OffsetNormalizedAge, B=4), B=1))
     )
     2=&LerpPosition(
      A=$Output.SocketLocation002.SampledPosition
      B=$Output.SocketLocation003.SampledPosition
      Alpha=&ClampFloat(Float=&Subtract_Float(A=&Multiply_Float(A=$Particles.OffsetNormalizedAge, B=4), B=2))
     )
     3=&LerpPosition(
      A=$Output.SocketLocation003.SampledPosition
      B=$Output.SocketLocation.SampledPosition
      Alpha=&ClampFloat(Float=&Subtract_Float(A=&Multiply_Float(A=$Particles.OffsetNormalizedAge, B=4), B=3))
     )
     Index=&MakeCustomIntFromBool(
      Bool=&SetBoolByFloatComparison(A=$Particles.OffsetNormalizedAge, B=0.25, Comparison Type=::NewEnumerator6)
      True INT=0
      False INT=&MakeCustomIntFromBool(
       Bool=&SetBoolByFloatComparison(A=$Particles.OffsetNormalizedAge, B=0.5, Comparison Type=::NewEnumerator6)
       True INT=1
       False INT=&MakeCustomIntFromBool(
        Bool=&SetBoolByFloatComparison(A=$Particles.OffsetNormalizedAge, B=0.75, Comparison Type=::NewEnumerator6)
        True INT=2
        False INT=3
       )
      )
     )
     Num=::NewEnumerator2
    )) {isSetParameters=true, module=SetVariables_#5}
   &FN_Light_Attributes(
    Create Exponent Attribute=true
    Create Radius Attribute=true
    Enable Particle Light Enable=true
    Exponent=2
    Light Enabled:NiagaraBool=&SetBoolByIntComparison(A=$Particles.PartitionParticles.Index In Partition, B=3, Comparison Type=::NewEnumerator4)
    Radius=8
   )
   /SetVariables(
    Particles.Capsule Color:LinearColor=&MultiplyLinearColor(
     Linear Color=$Particles.Color
     Scale Factor=&MakeVector4FromVectorAndFloat("Vector (RGB)"=&VectorFromFloat(Value=1), "Float (Alpha)"=0.15)
    )
    Particles.LightColor=&MultiplyLinearColor(Linear Color=$Particles.Color, Scale Factor=&MakeVector4_FromFloat(Float=1))
   ) {isSetParameters=true, module=SetVariables_#6}
   &DynamicMaterialParameters(
    Index 0 Param 1=&MakeFloatFromVector(
     Vector=&RotateVectorByQuaternion(Vector To Rotate=$Emitter.CapsuleGlowDirection, Quaternion=$Output.SocketLocation.SampledRotation)
     Channel=::NewEnumerator0
    )
    Index 0 Param 2=&MakeFloatFromVector(
     Vector=&RotateVectorByQuaternion(Vector To Rotate=$Emitter.CapsuleGlowDirection, Quaternion=$Output.SocketLocation.SampledRotation)
     Channel=::NewEnumerator1
    )
    Index 0 Param 3=&MakeFloatFromVector(
     Vector=&RotateVectorByQuaternion(Vector To Rotate=$Emitter.CapsuleGlowDirection, Quaternion=$Output.SocketLocation.SampledRotation)
     Channel=::NewEnumerator2
    )
    Index 0 Param 4=1.5
    Module.Param0WriteEnabled=true
    Module.Param1WriteEnabled=true
    Module.Param2WriteEnabled=true
    Module.Param3WriteEnabled=true
   )
  ]
  rendererClasses [&NiagaraLightRendererProperties, &NiagaraSpriteRendererProperties]
  renderers [
   {rendererClass=&NiagaraLightRendererProperties, rendererIndex=0}
   {rendererClass=&NiagaraSpriteRendererProperties, rendererIndex=1}
   {rendererClass=&NiagaraSpriteRendererProperties, rendererIndex=2}
  ]
 }
 Light_Strip_Glow {
  emitterSpawnScript [/SetVariables(Emitter.CapsuleGlowDirection:Vector3f=<1 0 0>) {isSetParameters=true}]
  emitterUpdateScript [
   &EmitterState(Inactive Response=::NewEnumerator1, Life Cycle Mode=::NewEnumerator1, Loop Behavior=::NewEnumerator1, Loop Duration=8.7, UseLoopDelay=false)
   &SpawnBurst_Instantaneous(Spawn Count=120, Use Spawn Probability=false, UseLoopCountLimit=false)
  ]
  particleSpawnScript [
   &InitializeParticle(Color=$System.Glow Color, Sprite Size Mode=::NewEnumerator3, Uniform Sprite Size=0.5, UsePositionOffset=false, Write Material Random=true) {module=InitializeParticle001}
   &CameraOffset()
   &PartitionParticles(Number Of Partitions=12, Particles Per Partition=10, Partition Mode:ENiagara_PartitionMode_v2=::NewEnumerator1)
   /SetVariables(Particles.SpriteSizeCapsule:Vector2f=<2 2>) {isSetParameters=true}
  ]
  particleUpdateScript [
   &ParticleState(Kill Particles When Lifetime Has Elapsed=false, Loop Particles Lifetime=true)
   &SocketLocation(
    Position Sampling=::NewEnumerator1
    SamplePosOffset=false
    Socket Index=4
    Socket Reader=$System.Socket Reader
    Socket Sampling Mode=::NewEnumerator3
    VelocitySpeedLimit=false
   )
   &SocketLocation(
    Position Sampling=::NewEnumerator1
    SamplePosOffset=false
    Socket Index=5
    Socket Reader=$System.Socket Reader
    Socket Sampling Mode=::NewEnumerator3
    VelocitySpeedLimit=false
   ) {module=SocketLocation001}
   &SocketLocation(
    Position Sampling=::NewEnumerator1
    SamplePosOffset=false
    Socket Index=6
    Socket Reader=$System.Socket Reader
    Socket Sampling Mode=::NewEnumerator3
    VelocitySpeedLimit=false
   ) {module=SocketLocation002}
   &SocketLocation(
    Position Sampling=::NewEnumerator1
    SamplePosOffset=false
    Socket Index=7
    Socket Reader=$System.Socket Reader
    Socket Sampling Mode=::NewEnumerator3
    VelocitySpeedLimit=false
   ) {module=SocketLocation003}
   &SocketLocation(
    Position Sampling=::NewEnumerator1
    SamplePosOffset=false
    Socket Index=8
    Socket Reader=$System.Socket Reader
    Socket Sampling Mode=::NewEnumerator3
    VelocitySpeedLimit=false
   ) {module=SocketLocation004}
   &SocketLocation(
    Position Sampling=::NewEnumerator1
    SamplePosOffset=false
    Socket Index=9
    Socket Reader=$System.Socket Reader
    Socket Sampling Mode=::NewEnumerator3
    VelocitySpeedLimit=false
   ) {module=SocketLocation005}
   &SocketLocation(
    Position Sampling=::NewEnumerator1
    SamplePosOffset=false
    Socket Index=10
    Socket Reader=$System.Socket Reader
    Socket Sampling Mode=::NewEnumerator3
    VelocitySpeedLimit=false
   ) {module=SocketLocation006}
   &SocketLocation(
    Position Sampling=::NewEnumerator1
    SamplePosOffset=false
    Socket Index=11
    Socket Reader=$System.Socket Reader
    Socket Sampling Mode=::NewEnumerator3
    VelocitySpeedLimit=false
   ) {module=SocketLocation007}
   /SetVariables(Particles.OffsetNormalizedAge:NiagaraFloat=&ModuloFloat(
     A=&Add_Float(
      A=$Particles.NormalizedAge
      B=&Add_Float(
       A=&MakeCustomFloatFromBool(
        Bool=&SetBoolByFloatComparison(A=$Particles.PartitionParticles.Normalized Index In Partition, B=0.5, Comparison Type=::NewEnumerator2)
        True Float=0.375
        False Float=0
       )
       B=&Multiply_Float(A=$Particles.PartitionParticles.Normalized Index In Partition, B=0.25)
      )
     )
     B=1
    )) {isSetParameters=true, module=SetVariables_#7}
   /SetVariables(Particles.Position=&MaskVector(
     Mask Source=1
     Mask Parameter=/Niagara/DynamicInputs/Divide/Divide_Int(A:NiagaraInt32=$Particles.PartitionParticles.Partition, B:NiagaraInt32=8)
     Mask=0
     Masked Vector=&MergeHelper_VectorSwitch(
      0=&Lerp_Vector(
       A=Linked() {linkedVariable {name=Output.SocketLocation.SampledPosition, type=NiagaraPosition}}
       B=Linked() {linkedVariable {name=Output.SocketLocation001.SampledPosition, type=NiagaraPosition}}
       Alpha=$Particles.OffsetNormalizedAge
      )
      1=&Lerp_Vector(
       A=Linked() {linkedVariable {name=Output.SocketLocation002.SampledPosition, type=NiagaraPosition}}
       B=Linked() {linkedVariable {name=Output.SocketLocation001.SampledPosition, type=NiagaraPosition}}
       Alpha=$Particles.OffsetNormalizedAge
      )
      2=&Lerp_Vector(
       A=Linked() {linkedVariable {name=Output.SocketLocation002.SampledPosition, type=NiagaraPosition}}
       B=Linked() {linkedVariable {name=Output.SocketLocation003.SampledPosition, type=NiagaraPosition}}
       Alpha=$Particles.OffsetNormalizedAge
      )
      3=&Lerp_Vector(
       A=Linked() {linkedVariable {name=Output.SocketLocation.SampledPosition, type=NiagaraPosition}}
       B=Linked() {linkedVariable {name=Output.SocketLocation003.SampledPosition, type=NiagaraPosition}}
       Alpha=$Particles.OffsetNormalizedAge
      )
      4=&Lerp_Vector(
       A=Linked() {linkedVariable {name=Output.SocketLocation005.SampledPosition, type=NiagaraPosition}}
       B=Linked() {linkedVariable {name=Output.SocketLocation004.SampledPosition, type=NiagaraPosition}}
       Alpha=$Particles.OffsetNormalizedAge
      )
      5=&Lerp_Vector(
       A=Linked() {linkedVariable {name=Output.SocketLocation005.SampledPosition, type=NiagaraPosition}}
       B=Linked() {linkedVariable {name=Output.SocketLocation006.SampledPosition, type=NiagaraPosition}}
       Alpha=$Particles.OffsetNormalizedAge
      )
      6=&Lerp_Vector(
       A=Linked() {linkedVariable {name=Output.SocketLocation007.SampledPosition, type=NiagaraPosition}}
       B=Linked() {linkedVariable {name=Output.SocketLocation006.SampledPosition, type=NiagaraPosition}}
       Alpha=$Particles.OffsetNormalizedAge
      )
      7=&Lerp_Vector(
       A=Linked() {linkedVariable {name=Output.SocketLocation007.SampledPosition, type=NiagaraPosition}}
       B=Linked() {linkedVariable {name=Output.SocketLocation004.SampledPosition, type=NiagaraPosition}}
       Alpha=$Particles.OffsetNormalizedAge
      )
      Index=$Particles.PartitionParticles.Partition
      Num=::NewEnumerator6
     )
     Passthrough Vector=&MergeHelper_VectorSwitch(
      0=&Lerp_Vector(
       A=Linked() {linkedVariable {name=Output.SocketLocation004.SampledPosition, type=NiagaraPosition}}
       B=Linked() {linkedVariable {name=Output.SocketLocation.SampledPosition, type=NiagaraPosition}}
       Alpha=$Particles.OffsetNormalizedAge
      )
      1=&Lerp_Vector(
       A=Linked() {linkedVariable {name=Output.SocketLocation001.SampledPosition, type=NiagaraPosition}}
       B=Linked() {linkedVariable {name=Output.SocketLocation005.SampledPosition, type=NiagaraPosition}}
       Alpha=$Particles.OffsetNormalizedAge
      )
      2=&Lerp_Vector(
       A=Linked() {linkedVariable {name=Output.SocketLocation006.SampledPosition, type=NiagaraPosition}}
       B=Linked() {linkedVariable {name=Output.SocketLocation002.SampledPosition, type=NiagaraPosition}}
       Alpha=$Particles.OffsetNormalizedAge
      )
      3=&Lerp_Vector(
       A=Linked() {linkedVariable {name=Output.SocketLocation003.SampledPosition, type=NiagaraPosition}}
       B=Linked() {linkedVariable {name=Output.SocketLocation007.SampledPosition, type=NiagaraPosition}}
       Alpha=$Particles.OffsetNormalizedAge
      )
      Index=&ModuloInt(A=$Particles.PartitionParticles.Partition, B=8)
      Num=::NewEnumerator2
     )
    )) {isSetParameters=true, module=SetVariables_#5}
   &ScaleColor(
    Scale Alpha=&FloatFromCurve(FloatCurve=DataInterface() {propertyValues {Curve {keys [
         {arriveTangent=4.1785, interpMode=RCIM_Cubic, leaveTangent=4.1785, tangentMode=RCTM_User, time=0, value=0}
         {arriveTangent=1.5009, interpMode=RCIM_Cubic, leaveTangent=1.5009, tangentMode=RCTM_User, time=0.2, value=0.55}
         {arriveTangent=0.0224, interpMode=RCIM_Cubic, leaveTangent=0.0224, tangentMode=RCTM_User, time=1, value=1}
        ]}}})
    Scale RGB=&VectorFromFloat(Value=&FloatFromCurve(
      FloatCurve=DataInterface() {propertyValues {Curve {
         defaultValue=0
         keys [
          <at 0 0 RCIM_Cubic>
          {arriveTangent=0.0159, interpMode=RCIM_Cubic, leaveTangent=0.0159, tangentMode=RCTM_User, time=0.088, value=1}
          {interpMode=RCIM_Cubic, tangentMode=RCTM_User, time=0.9, value=1}
          {interpMode=RCIM_Cubic, tangentMode=RCTM_User, time=1, value=0}
         ]
        }}}
      CurveIndex=$Emitter.NormalizedLoopAge
      Scale Curve=3
     ))
    ScaleA=false
    ScaleRGB=true
   )
   /SetVariables(
    Particles.Capsule Color:LinearColor=&MultiplyLinearColor(
     Linear Color=$Particles.Color
     Scale Factor=&MakeVector4FromVectorAndFloat("Vector (RGB)"=&VectorFromFloat(Value=1), "Float (Alpha)"=0.15)
    )
    Particles.LightColor=&MultiplyLinearColor(Linear Color=$Particles.Color, Scale Factor=&MakeVector4_FromFloat(Float=2))
   ) {isSetParameters=true}
   &FN_Light_Attributes(Create Exponent Attribute=true, Create Radius Attribute=true, Exponent=5, Radius=2)
   &DynamicMaterialParameters(
    Index 0 Param 1=&MakeFloatFromVector(
     Vector=&RotateVectorByQuaternion(Vector To Rotate=$Emitter.CapsuleGlowDirection, Quaternion=$Output.SocketLocation.SampledRotation)
     Channel=::NewEnumerator0
    )
    Index 0 Param 2=&MakeFloatFromVector(
     Vector=&RotateVectorByQuaternion(Vector To Rotate=$Emitter.CapsuleGlowDirection, Quaternion=$Output.SocketLocation.SampledRotation)
     Channel=::NewEnumerator1
    )
    Index 0 Param 3=&MakeFloatFromVector(
     Vector=&RotateVectorByQuaternion(Vector To Rotate=$Emitter.CapsuleGlowDirection, Quaternion=$Output.SocketLocation.SampledRotation)
     Channel=::NewEnumerator2
    )
    Index 0 Param 4=2
    Module.Param0WriteEnabled=true
    Module.Param1WriteEnabled=true
    Module.Param2WriteEnabled=true
    Module.Param3WriteEnabled=true
   )
  ]
  rendererClasses [&NiagaraSpriteRendererProperties]
  renderers [
   {rendererClass=&NiagaraSpriteRendererProperties}
   {rendererClass=&NiagaraSpriteRendererProperties, rendererIndex=1}
  ]
 }
}
summary {
 systemName=NS_Backpack_DryEraseToro_GlowLights
 emitters [
  {emitterName=Glass_Glow, rendererClasses [&NiagaraLightRendererProperties, &NiagaraSpriteRendererProperties]}
  {emitterName=Light_Strip_Glow, rendererClasses [&NiagaraSpriteRendererProperties]}
  {bEnabled=false, emitterName=CapsoleGlow001, rendererClasses [&NiagaraSpriteRendererProperties]}
  {emitterName=Door_Glow, rendererClasses [&NiagaraLightRendererProperties, &NiagaraSpriteRendererProperties]}
  {emitterName=Arrow_Glow, rendererClasses [&NiagaraSpriteRendererProperties]}
 ]
 userVariables [
  User.bIsFrontEnd {type=NiagaraBool}
  User.bIsFrontEndPreview {type=NiagaraBool}
 ]
}
