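# Material function for the DryEraseToro back bling that adds an elimination-driven, flickering emissive glow to incoming material attributes.
# An FX mask texture blends two light colors, tinting base color, with a time-based sine flicker offset per pixel.
# Its green channel, scaled by Elims, blends the glow over existing emissive.
# When Elims_end equals one, a timestamp animation replaces that scale with a half-second fade-out.
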MaterialFunction /BRCosmetics/Accessories/FORT_Backpacks/Backpack_M_MED_DryEraseToro/Materials/MF_DryEraseToro_BackPack
graph {nodes [
  n1 = GetMaterialAttributes(None=FunctionInput() {inputType=FunctionInput_MaterialAttributes}) {attributeGetTypes [#1, #2]}
  n2 = VectorParameter() {group=DryEraseToro, parameterName=LightColor, defaultValue=<rgba 1 0.8 0.8 1>}
  n3 = VectorParameter() {group=DryEraseToro, parameterName=LightColor2, defaultValue=<rgba 15 0.8998 0.15 1>}
  n4 = TextureSample() {
   samplerType=SAMPLERTYPE_Masks
   texture=/BRCosmetics/Accessories/FORT_Backpacks/Backpack_M_MED_DryEraseToro/Textures/T_DryEraseToro_Backpack_FX2
  }
  n5 = Power(Base=Multiply(A=@n4.B) {constB=2})
  n6 = Add(A=Multiply(A=Time(), B=ScalarParameter() {defaultValue=1.1, group=DryEraseToro, parameterName=FlickSpeed}) {constB=0.9}, B=Power(Base=@n4.R))
  n7 = Multiply(A=@n1.BaseColor, B=Multiply(A=LinearInterpolate(A=@n2.RGB, B=@n3.RGB, Alpha=@n5), B=Sine(None=@n6)))
  n8 = Multiply(A=@n4.G, B=ScalarParameter() {group=DryEraseToro, parameterName=Elims})
  n9 = MaterialFunctionCall!(TimeStart=ScalarParameter() {parameterName=Time Start}, AnimLength=Constant() {r=0.5}, TimeCurrent) {materialFunction=/Game/Global/MaterialFunctions/TechArt/MF_TimeStampAnim}
  n10 = If(
   A=ScalarParameter() {group=DryEraseToro, parameterName=Elims_end}
   "A > B"=@n8
   "A == B"=Multiply(A=@n4.G, B=OneMinus(None=@n9.AnimValue))
   "A < B"=@n8
  ) {constB=1}
  FunctionOutput(None=SetMaterialAttributes!(MaterialAttributes=@n1.MaterialAttributes, EmissiveColor=LinearInterpolate(A=@n1.EmissiveColor, B=@n7, Alpha=@n10)) {attributeSetTypes [#1]})
 ]}
