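# Material function for the DryEraseToro backpack that outputs a rotating mask with two narrow, opposed radial bands.
# UVs shifted by a U/V center offset spin through CustomRotator at Time times RotateSpeed, then convert to radial coordinates.
# The nearer distance from the first radial component to two fixed positions, minus Width, is steeply scaled and saturated, leaving hard-edged dark bands on white.
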MaterialFunction /BRCosmetics/Accessories/FORT_Backpacks/Backpack_M_MED_DryEraseToro/Materials/MF_DryEraseToro_Rotator_BackPack
graph {nodes [
  n1 = FunctionInput(Preview=ScalarParameter() {defaultValue=0.2, parameterName=UV_Center_OffsetU, sortPriority=1}) {bUsePreviewValueAsDefault=true, inputName=UV_Center_OffsetU, inputType=FunctionInput_Scalar, sortPriority=1}
  n2 = FunctionInput(Preview=ScalarParameter() {defaultValue=0.2, parameterName=UV_Center_OffsetV, sortPriority=1}) {bUsePreviewValueAsDefault=true, inputName=UV_Center_OffsetV, inputType=FunctionInput_Scalar, sortPriority=1}
  n3 = ScalarParameter() {defaultValue=0.1, parameterName=RotateSpeed}
  n4 = MaterialFunctionCall(
   UVs=Add(A=TextureCoordinate(), B=AppendVector(A=@n1, B=@n2)) {constB=0.2}
   "Rotation Angle (0-1)"=Multiply(A=Time(), B=FunctionInput(Preview=@n3) {bUsePreviewValueAsDefault=true, inputName=RotateSpeed, inputType=FunctionInput_Scalar}) {constB=-0.3}
  ) {materialFunction=&CustomRotator}
  n5 = ComponentMask(None=MaterialFunctionCall(Vector or UVs=Subtract(A=@n4.Rotated Values) {constB=0.5}) {materialFunction=&VectorToRadialValue}.Radial Coordinates) {r=true}
  n6 = Constant() {r=0.25}
  n7 = Constant() {r=0.75}
  n8 = Subtract(
   A=Min(A=Distance(A=@n5, B=@n6), B=Distance(A=@n5, B=@n7))
   B=FunctionInput(Preview=ScalarParameter() {defaultValue=0.05, parameterName=Width}) {bUsePreviewValueAsDefault=true, inputName=Width, inputType=FunctionInput_Scalar}
  ) {constB=0.1}
  FunctionOutput(None=Saturate(None=Multiply(A=@n8) {constB=50}))
 ]}
