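# Animation Blueprint for the DryEraseToro backpack, adding jiggle to eight dolls and an elimination reaction.
# The pack bone offsets from spine bend and skydive state, then chained Anim Dynamics simulate each doll's mount, body and head into a cached pose.
# When the part modifier enables reactivity, enemy eliminations or a looping frontend timer trigger a timed sequence: random-rotator shake on the first doll, then constraining it to the door bone.
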AnimBlueprint /BRCosmetics/Accessories/FORT_Backpacks/Backpack_M_MED_DryEraseToro/Meshes/M_MED_DryEraseToro_Pack_AnimBP
target_skeleton=&/M_MED_DryEraseToro_Pack_Skeleton
graphs [
 AnimGraph {nodes [
   n1 = "Constraint - dyn_doll_UP_0 (P)"!(
    ComponentPose=Local To Component(LocalPose=Use cached pose 'anim dynamics'())
    bAlphaBoolEnabled:bool=Get ElimReactive() -> (ElimReactive:bool)
    ConstraintWeights_0=1
   ) -> (Pose) {
    alphaInputType=Bool
    alphaBoolBlend {blendOption=Linear}
    boneToModify {boneName=dyn_doll_UP_0}
    constraintSetup [{offsetOption=None, targetBone {boneName=dyn_door}}]
    constraintWeights [1]
   }
   n2 = "Transform (Modify) Bone - Bone: dyn_doll_UP_0"!(
    ComponentPose=@n1
    bAlphaBoolEnabled:bool=Get Shaking() -> (Shaking:bool)
    Translation="0,0,0"
    Rotation=Get Doll Shake R() -> (Doll Shake R:Rotator)
    Scale="1,1,1"
   ) -> (Pose) {
    alphaInputType=Bool
    rotationMode=BMM_Additive
    rotationSpace=BCS_BoneSpace
    translationMode=BMM_Additive
    translationSpace=BCS_BoneSpace
    alphaBoolBlend {blendInTime=0.2, blendOption=QuarticInOut, blendOutTime=0.8}
    boneToModify {boneName=dyn_doll_UP_0}
   }
   Output Pose(Result=Blend Poses by bool(
     BlendPose_0=Component To Local(ComponentPose=@n2)
     BlendPose_1=Use cached pose 'anim dynamics'()
     bActiveValue=Get test reactivity() -> (test reactivity:bool)
    ) {blendType=Linear}) {layerGroup=None}
   n3 = "Transform (Modify) Bone - Bone: pack"(ComponentPose=Local To Component(LocalPose=Input Pose()), Translation=Get backpack_translation(), Rotation=Get backpack_rotate()) {
    rotationMode=BMM_Additive
    rotationSpace=BCS_BoneSpace
    translationMode=BMM_Additive
    translationSpace=BCS_BoneSpace
    alphaBoolBlend {blendOption=Linear}
    boneToModify {boneName=pack}
   }
   n4 = Anim Dynamics - Bone: dyn_doll_UP_0(ComponentPose=@n3, Alpha=1, ExternalForce="0,0,0") {
    angularSpringConstant=3000
    bLinearSpring=true
    bUseGravityOverride=true
    linearSpringConstant=7
    alphaBoolBlend {blendOption=Linear}
    boundBone {boneName=dyn_doll_UP_0}
    componentAppliedLinearAccClamp=<1200 1200 1200>
    componentLinearAccScale=<0.5 0.5 0.5>
    componentLinearVelScale=<0.6 0.6 0.6>
    gravityOverride {z=-980}
    physicsBodyDefinitions [{
      boundBone {boneName=dyn_doll_UP_0}
      boxExtents=<10 10 5>
      constraintSetup {angularLimitsMax=<0 0 0>, angularLimitsMin=<0 0 0>, angularTarget=<1 0 0>, linearAxesMax=<0.45 0 0>}
      localJointOffset=<0.5 0 0>
     }]
   }
   n5 = Anim Dynamics - Bone: dyn_doll_0(ComponentPose=@n4, Alpha=1, ExternalForce="0,0,0") {
    angularDampingOverride=0.999
    angularSpringConstant=3000
    bAngularSpring=true
    bOverrideAngularDamping=true
    bUseGravityOverride=true
    linearSpringConstant=5000
    alphaBoolBlend {blendOption=Linear}
    boundBone {boneName=dyn_doll_0}
    componentAppliedLinearAccClamp=<1000 1000 1000>
    componentLinearAccScale=<0.4 4 0.4>
    componentLinearVelScale=<0.4 0.4 0.4>
    gravityOverride {z=-980}
    physicsBodyDefinitions [{
      boundBone {boneName=dyn_doll_0}
      boxExtents=<10 5 5>
      constraintSetup {angularLimitsMax=<0 10 5>, angularLimitsMin=<0 -10 -5>, angularTarget=<1 0 0>}
      localJointOffset=<0.5 0 0>
     }]
   }
   n6 = Anim Dynamics - Bone: dyn_dool_head_0(ComponentPose=@n5, Alpha=1, ExternalForce="0,0,0") {
    angularDampingOverride=0.9999
    angularSpringConstant=2000
    bAngularSpring=true
    bOverrideAngularDamping=true
    bUseGravityOverride=true
    linearSpringConstant=5000
    alphaBoolBlend {blendOption=Linear}
    boundBone {boneName=dyn_dool_head_0}
    componentAppliedLinearAccClamp=<800 800 800>
    componentLinearAccScale=<0.2 0.2 0.2>
    componentLinearVelScale=<0.3 0.3 0.3>
    gravityOverride {z=-980}
    physicsBodyDefinitions [{
      boundBone {boneName=dyn_dool_head_0}
      boxExtents=<7 7 7>
      constraintSetup {angularLimitsMax=<0 10 10>, angularLimitsMin=<0 -10 -10>, angularTarget=<1 0 0>}
      localJointOffset=<0.5 0 0>
     }]
   }
   n7 = Anim Dynamics - Bone: dyn_doll_UP_1(ComponentPose=@n6, Alpha=1, ExternalForce="0,0,0") {
    angularSpringConstant=3000
    bLinearSpring=true
    bUseGravityOverride=true
    linearSpringConstant=5
    alphaBoolBlend {blendOption=Linear}
    boundBone {boneName=dyn_doll_UP_1}
    componentAppliedLinearAccClamp=<1200 1200 1200>
    componentLinearAccScale=<0.5 0.5 0.5>
    componentLinearVelScale=<0.6 0.6 0.6>
    gravityOverride {z=-980}
    physicsBodyDefinitions [{
      boundBone {boneName=dyn_doll_UP_1}
      boxExtents=<10 10 5>
      constraintSetup {angularLimitsMax=<0 0 0>, angularLimitsMin=<0 0 0>, angularTarget=<1 0 0>, linearAxesMax=<0.45 0 0>}
      localJointOffset=<0.8 0 0>
     }]
   }
   n8 = Anim Dynamics - Bone: dyn_doll_1(ComponentPose=@n7, Alpha=1, ExternalForce="0,0,0") {
    angularDampingOverride=0.999
    angularSpringConstant=3000
    bAngularSpring=true
    bOverrideAngularDamping=true
    bUseGravityOverride=true
    linearSpringConstant=5000
    alphaBoolBlend {blendOption=Linear}
    boundBone {boneName=dyn_doll_1}
    componentAppliedLinearAccClamp=<1000 1000 1000>
    componentLinearAccScale=<0.3 0.3 0.3>
    componentLinearVelScale=<0.4 0.4 0.4>
    gravityOverride {z=-980}
    physicsBodyDefinitions [{
      boundBone {boneName=dyn_doll_1}
      boxExtents=<10 5 5>
      constraintSetup {angularLimitsMax=<0 15 5>, angularLimitsMin=<0 -15 -5>, angularTarget=<1 0 0>}
      localJointOffset=<0.25 0 0>
     }]
   }
   n9 = Anim Dynamics - Bone: dyn_dool_head_1(ComponentPose=@n8, Alpha=1, ExternalForce="0,0,0") {
    angularDampingOverride=1
    angularSpringConstant=2000
    bAngularSpring=true
    bOverrideAngularDamping=true
    bUseGravityOverride=true
    linearSpringConstant=5000
    alphaBoolBlend {blendOption=Linear}
    boundBone {boneName=dyn_dool_head_1}
    componentAppliedLinearAccClamp=<1000 1000 1000>
    componentLinearAccScale=<0.2 0.2 0.2>
    componentLinearVelScale=<0.2 0.2 0.2>
    gravityOverride {z=-980}
    physicsBodyDefinitions [{
      boundBone {boneName=dyn_dool_head_1}
      boxExtents=<7 7 7>
      constraintSetup {angularLimitsMax=<0 10 10>, angularLimitsMin=<0 -10 -10>, angularTarget=<1 0 0>}
      localJointOffset=<0.3 0 0>
     }]
   }
   n10 = Anim Dynamics - Bone: dyn_doll_UP_2(ComponentPose=@n9, Alpha=1, ExternalForce="0,0,0") {
    angularSpringConstant=3000
    bLinearSpring=true
    bUseGravityOverride=true
    linearSpringConstant=10
    alphaBoolBlend {blendOption=Linear}
    boundBone {boneName=dyn_doll_UP_2}
    componentAppliedLinearAccClamp=<1200 1200 1200>
    componentLinearAccScale=<0.5 0.5 0.5>
    componentLinearVelScale=<0.6 0.6 0.6>
    gravityOverride {z=-980}
    physicsBodyDefinitions [{
      boundBone {boneName=dyn_doll_UP_2}
      boxExtents=<10 10 5>
      constraintSetup {angularLimitsMax=<0 0 0>, angularLimitsMin=<0 0 0>, angularTarget=<1 0 0>, linearAxesMax=<0.3 0 0>}
      localJointOffset=<0.7 0 0>
     }]
   }
   n11 = Anim Dynamics - Bone: dyn_doll_2(ComponentPose=@n10, Alpha=1, ExternalForce="0,0,0") {
    angularDampingOverride=0.999
    angularSpringConstant=3000
    bAngularSpring=true
    bOverrideAngularDamping=true
    bUseGravityOverride=true
    linearSpringConstant=5000
    alphaBoolBlend {blendOption=Linear}
    boundBone {boneName=dyn_doll_2}
    componentAppliedLinearAccClamp=<1000 1000 1000>
    componentLinearAccScale=<0.4 4 0.4>
    componentLinearVelScale=<0.4 0.4 0.4>
    gravityOverride {z=-980}
    physicsBodyDefinitions [{
      boundBone {boneName=dyn_doll_2}
      boxExtents=<10 5 5>
      constraintSetup {angularLimitsMax=<0 10 5>, angularLimitsMin=<0 -10 -5>, angularTarget=<1 0 0>}
      localJointOffset=<0.5 0 0>
     }]
   }
   n12 = Anim Dynamics - Bone: dyn_dool_head_2(ComponentPose=@n11, Alpha=1, ExternalForce="0,0,0") {
    angularDampingOverride=0.9999
    angularSpringConstant=2000
    bAngularSpring=true
    bOverrideAngularDamping=true
    bUseGravityOverride=true
    linearSpringConstant=5000
    alphaBoolBlend {blendOption=Linear}
    boundBone {boneName=dyn_dool_head_2}
    componentAppliedLinearAccClamp=<1000 1000 1000>
    componentLinearAccScale=<0.2 0.2 0.2>
    componentLinearVelScale=<0.3 0.3 0.3>
    gravityOverride {z=-980}
    physicsBodyDefinitions [{
      boundBone {boneName=dyn_dool_head_2}
      boxExtents=<7 7 7>
      constraintSetup {angularLimitsMax=<0 10 15>, angularLimitsMin=<0 -10 -15>, angularTarget=<1 0 0>}
      localJointOffset=<0.5 0 0>
     }]
   }
   n13 = Anim Dynamics - Bone: dyn_doll_UP_3(ComponentPose=@n12, Alpha=1, ExternalForce="0,0,0") {
    angularSpringConstant=3000
    bLinearSpring=true
    bUseGravityOverride=true
    linearSpringConstant=5
    alphaBoolBlend {blendOption=Linear}
    boundBone {boneName=dyn_doll_UP_3}
    componentAppliedLinearAccClamp=<1200 1200 1200>
    componentLinearAccScale=<0.5 0.5 0.5>
    componentLinearVelScale=<0.6 0.6 0.6>
    gravityOverride {z=-980}
    physicsBodyDefinitions [{
      boundBone {boneName=dyn_doll_UP_3}
      boxExtents=<10 10 5>
      constraintSetup {angularLimitsMax=<0 0 0>, angularLimitsMin=<0 0 0>, angularTarget=<1 0 0>, linearAxesMax=<0.4 0 0>}
      localJointOffset=<0.8 0 0>
     }]
   }
   n14 = Anim Dynamics - Bone: dyn_doll_3(ComponentPose=@n13, Alpha=1, ExternalForce="0,0,0") {
    angularDampingOverride=0.999
    angularSpringConstant=3000
    bAngularSpring=true
    bOverrideAngularDamping=true
    bUseGravityOverride=true
    linearSpringConstant=5000
    alphaBoolBlend {blendOption=Linear}
    boundBone {boneName=dyn_doll_3}
    componentAppliedLinearAccClamp=<1000 1000 1000>
    componentLinearAccScale=<0.3 0.3 0.3>
    componentLinearVelScale=<0.4 0.4 0.4>
    gravityOverride {z=-980}
    physicsBodyDefinitions [{
      boundBone {boneName=dyn_doll_3}
      boxExtents=<10 5 5>
      constraintSetup {angularLimitsMax=<0 15 5>, angularLimitsMin=<0 -15 -5>, angularTarget=<1 0 0>}
      localJointOffset=<0.25 0 0>
     }]
   }
   n15 = Anim Dynamics - Bone: dyn_dool_head_3(ComponentPose=@n14, Alpha=1, ExternalForce="0,0,0") {
    angularDampingOverride=1
    angularSpringConstant=2000
    bAngularSpring=true
    bOverrideAngularDamping=true
    bUseGravityOverride=true
    linearSpringConstant=5000
    alphaBoolBlend {blendOption=Linear}
    boundBone {boneName=dyn_dool_head_3}
    componentAppliedLinearAccClamp=<1000 1000 1000>
    componentLinearAccScale=<0.2 0.2 0.2>
    componentLinearVelScale=<0.2 0.2 0.2>
    gravityOverride {z=-980}
    physicsBodyDefinitions [{
      boundBone {boneName=dyn_dool_head_3}
      boxExtents=<7 7 7>
      constraintSetup {angularLimitsMax=<0 10 10>, angularLimitsMin=<0 -10 -10>, angularTarget=<1 0 0>}
      localJointOffset=<0.3 0 0>
     }]
   }
   n16 = Anim Dynamics - Bone: dyn_doll_UP_4(ComponentPose=@n15, Alpha=1, ExternalForce="0,0,0") {
    angularSpringConstant=3000
    bLinearSpring=true
    bUseGravityOverride=true
    linearSpringConstant=10
    alphaBoolBlend {blendOption=Linear}
    boundBone {boneName=dyn_doll_UP_4}
    componentAppliedLinearAccClamp=<1200 1200 1200>
    componentLinearAccScale=<0.5 0.5 0.5>
    componentLinearVelScale=<0.6 0.6 0.6>
    gravityOverride {z=-980}
    physicsBodyDefinitions [{
      boundBone {boneName=dyn_doll_UP_4}
      boxExtents=<10 10 5>
      constraintSetup {angularLimitsMax=<0 0 0>, angularLimitsMin=<0 0 0>, angularTarget=<1 0 0>, linearAxesMax=<0.3 0 0>}
      localJointOffset=<0.7 0 0>
     }]
   }
   n17 = Anim Dynamics - Bone: dyn_doll_4(ComponentPose=@n16, Alpha=1, ExternalForce="0,0,0") {
    angularDampingOverride=0.999
    angularSpringConstant=3000
    bAngularSpring=true
    bOverrideAngularDamping=true
    bUseGravityOverride=true
    linearSpringConstant=5000
    alphaBoolBlend {blendOption=Linear}
    boundBone {boneName=dyn_doll_4}
    componentAppliedLinearAccClamp=<1000 1000 1000>
    componentLinearAccScale=<0.3 0.3 0.3>
    componentLinearVelScale=<0.4 0.4 0.4>
    gravityOverride {z=-980}
    physicsBodyDefinitions [{
      boundBone {boneName=dyn_doll_4}
      boxExtents=<10 5 5>
      constraintSetup {angularLimitsMax=<0 10 5>, angularLimitsMin=<0 -10 -5>, angularTarget=<1 0 0>}
      localJointOffset=<0.3 0 0>
     }]
   }
   n18 = Anim Dynamics - Bone: dyn_dool_head_4(ComponentPose=@n17, Alpha=1, ExternalForce="0,0,0") {
    angularDampingOverride=0.9999
    angularSpringConstant=2000
    bAngularSpring=true
    bOverrideAngularDamping=true
    bUseGravityOverride=true
    linearSpringConstant=5000
    alphaBoolBlend {blendOption=Linear}
    boundBone {boneName=dyn_dool_head_4}
    componentAppliedLinearAccClamp=<1000 1000 1000>
    componentLinearAccScale=<0.2 0.2 0.2>
    componentLinearVelScale=<0.3 0.3 0.3>
    gravityOverride {z=-980}
    physicsBodyDefinitions [{
      boundBone {boneName=dyn_dool_head_4}
      boxExtents=<7 7 7>
      constraintSetup {angularLimitsMax=<0 10 10>, angularLimitsMin=<0 -10 -10>, angularTarget=<1 0 0>}
      localJointOffset=<0.5 0 0>
     }]
   }
   n19 = Anim Dynamics - Bone: dyn_doll_UP_5(ComponentPose=@n18, Alpha=1, ExternalForce="0,0,0") {
    angularSpringConstant=3000
    bLinearSpring=true
    bUseGravityOverride=true
    linearSpringConstant=5
    alphaBoolBlend {blendOption=Linear}
    boundBone {boneName=dyn_doll_UP_5}
    componentAppliedLinearAccClamp=<1200 1200 1200>
    componentLinearAccScale=<0.5 0.5 0.5>
    componentLinearVelScale=<0.6 0.6 0.6>
    gravityOverride {z=-980}
    physicsBodyDefinitions [{
      boundBone {boneName=dyn_doll_UP_5}
      boxExtents=<10 10 5>
      constraintSetup {angularLimitsMax=<0 0 0>, angularLimitsMin=<0 0 0>, angularTarget=<1 0 0>, linearAxesMax=<0.45 0 0>}
      localJointOffset=<0.7 0 0>
     }]
   }
   n20 = Anim Dynamics - Bone: dyn_doll_5(ComponentPose=@n19, Alpha=1, ExternalForce="0,0,0") {
    angularDampingOverride=0.999
    angularSpringConstant=3000
    bAngularSpring=true
    bOverrideAngularDamping=true
    bUseGravityOverride=true
    linearSpringConstant=5000
    alphaBoolBlend {blendOption=Linear}
    boundBone {boneName=dyn_doll_5}
    componentAppliedLinearAccClamp=<1000 1000 1000>
    componentLinearAccScale=<0.4 4 0.4>
    componentLinearVelScale=<0.4 0.4 0.4>
    gravityOverride {z=-980}
    physicsBodyDefinitions [{
      boundBone {boneName=dyn_doll_5}
      boxExtents=<10 5 5>
      constraintSetup {angularLimitsMax=<0 15 5>, angularLimitsMin=<0 -15 -5>, angularTarget=<1 0 0>}
      localJointOffset=<0.5 0 0>
     }]
   }
   n21 = Anim Dynamics - Bone: dyn_dool_head_5(ComponentPose=@n20, Alpha=1, ExternalForce="0,0,0") {
    angularDampingOverride=1
    angularSpringConstant=2000
    bAngularSpring=true
    bOverrideAngularDamping=true
    bUseGravityOverride=true
    linearSpringConstant=5000
    alphaBoolBlend {blendOption=Linear}
    boundBone {boneName=dyn_dool_head_5}
    componentAppliedLinearAccClamp=<1000 1000 1000>
    componentLinearAccScale=<0.2 0.2 0.2>
    componentLinearVelScale=<0.2 0.2 0.2>
    gravityOverride {z=-980}
    physicsBodyDefinitions [{
      boundBone {boneName=dyn_dool_head_5}
      boxExtents=<7 7 7>
      constraintSetup {angularLimitsMax=<0 10 10>, angularLimitsMin=<0 -10 -10>, angularTarget=<1 0 0>}
      localJointOffset=<0.3 0 0>
     }]
   }
   n22 = Anim Dynamics - Bone: dyn_doll_UP_6(ComponentPose=@n21, Alpha=1, ExternalForce="0,0,0") {
    angularSpringConstant=3000
    bLinearSpring=true
    bUseGravityOverride=true
    linearSpringConstant=7
    alphaBoolBlend {blendOption=Linear}
    boundBone {boneName=dyn_doll_UP_6}
    componentAppliedLinearAccClamp=<1200 1200 1200>
    componentLinearAccScale=<0.5 0.5 0.5>
    componentLinearVelScale=<0.6 0.6 0.6>
    gravityOverride {z=-980}
    physicsBodyDefinitions [{
      boundBone {boneName=dyn_doll_UP_6}
      boxExtents=<10 10 5>
      constraintSetup {angularLimitsMax=<0 0 0>, angularLimitsMin=<0 0 0>, angularTarget=<1 0 0>, linearAxesMax=<0.4 0 0>}
      localJointOffset=<0.8 0 0>
     }]
   }
   n23 = Anim Dynamics - Bone: dyn_doll_6(ComponentPose=@n22, Alpha=1, ExternalForce="0,0,0") {
    angularDampingOverride=0.999
    angularSpringConstant=3000
    bAngularSpring=true
    bOverrideAngularDamping=true
    bUseGravityOverride=true
    linearSpringConstant=5000
    alphaBoolBlend {blendOption=Linear}
    boundBone {boneName=dyn_doll_6}
    componentAppliedLinearAccClamp=<1000 1000 1000>
    componentLinearAccScale=<0.3 0.3 0.3>
    componentLinearVelScale=<0.4 0.4 0.4>
    gravityOverride {z=-980}
    physicsBodyDefinitions [{
      boundBone {boneName=dyn_doll_6}
      boxExtents=<10 5 5>
      constraintSetup {angularLimitsMax=<0 10 5>, angularLimitsMin=<0 -10 -5>, angularTarget=<1 0 0>}
      localJointOffset=<0.3 0 0>
     }]
   }
   n24 = Anim Dynamics - Bone: dyn_dool_head_6(ComponentPose=@n23, Alpha=1, ExternalForce="0,0,0") {
    angularDampingOverride=1
    angularSpringConstant=2000
    bAngularSpring=true
    bOverrideAngularDamping=true
    bUseGravityOverride=true
    linearSpringConstant=5000
    alphaBoolBlend {blendOption=Linear}
    boundBone {boneName=dyn_dool_head_6}
    componentAppliedLinearAccClamp=<1000 1000 1000>
    componentLinearAccScale=<0.2 0.2 0.2>
    componentLinearVelScale=<0.3 0.3 0.3>
    gravityOverride {z=-980}
    physicsBodyDefinitions [{
      boundBone {boneName=dyn_dool_head_6}
      boxExtents=<7 7 7>
      constraintSetup {angularLimitsMax=<0 10 15>, angularLimitsMin=<0 -10 -15>, angularTarget=<1 0 0>}
      localJointOffset=<0.5 0 0>
     }]
   }
   n25 = Anim Dynamics - Bone: dyn_doll_UP_7(ComponentPose=@n24, Alpha=1, ExternalForce="0,0,0") {
    angularSpringConstant=3000
    bLinearSpring=true
    bUseGravityOverride=true
    linearSpringConstant=10
    alphaBoolBlend {blendOption=Linear}
    boundBone {boneName=dyn_doll_UP_7}
    componentAppliedLinearAccClamp=<1200 1200 1200>
    componentLinearAccScale=<0.5 0.5 0.5>
    componentLinearVelScale=<0.6 0.6 0.6>
    gravityOverride {z=-980}
    physicsBodyDefinitions [{
      boundBone {boneName=dyn_doll_UP_7}
      boxExtents=<10 10 5>
      constraintSetup {angularLimitsMax=<0 0 0>, angularLimitsMin=<0 0 0>, angularTarget=<1 0 0>, linearAxesMax=<0.3 0 0>}
      localJointOffset=<0.7 0 0>
     }]
   }
   n26 = Anim Dynamics - Bone: dyn_doll_7(ComponentPose=@n25, Alpha=1, ExternalForce="0,0,0") {
    angularDampingOverride=0.999
    angularSpringConstant=3000
    bAngularSpring=true
    bOverrideAngularDamping=true
    bUseGravityOverride=true
    linearSpringConstant=5000
    alphaBoolBlend {blendOption=Linear}
    boundBone {boneName=dyn_doll_7}
    componentAppliedLinearAccClamp=<1000 1000 1000>
    componentLinearAccScale=<0.4 4 0.4>
    componentLinearVelScale=<0.4 0.4 0.4>
    gravityOverride {z=-980}
    physicsBodyDefinitions [{
      boundBone {boneName=dyn_doll_7}
      boxExtents=<10 5 5>
      constraintSetup {angularLimitsMax=<0 10 5>, angularLimitsMin=<0 -10 -5>, angularTarget=<1 0 0>}
      localJointOffset=<0.5 0 0>
     }]
   }
   n27 = Anim Dynamics - Bone: dyn_dool_head_7(ComponentPose=@n26, Alpha=1, ExternalForce="0,0,0") {
    angularDampingOverride=0.9999
    angularSpringConstant=2000
    bAngularSpring=true
    bOverrideAngularDamping=true
    bUseGravityOverride=true
    linearSpringConstant=5000
    alphaBoolBlend {blendOption=Linear}
    boundBone {boneName=dyn_dool_head_7}
    componentAppliedLinearAccClamp=<800 800 800>
    componentLinearAccScale=<0.2 0.2 0.2>
    componentLinearVelScale=<0.2 0.2 0.2>
    gravityOverride {z=-980}
    physicsBodyDefinitions [{
      boundBone {boneName=dyn_dool_head_7}
      boxExtents=<7 7 7>
      constraintSetup {angularLimitsMax=<0 10 10>, angularLimitsMin=<0 -10 -10>, angularTarget=<1 0 0>}
      localJointOffset=<0.3 0 0>
     }]
   }
   n28 = "Transform (Modify) Bone - Bone: pack"!(ComponentPose=@n27, bAlphaBoolEnabled:bool=Get bIsSkydiveDiveMode(), Translation="0,-2,-2", Rotation="0,0,0", Scale) -> (Pose) {
    alphaInputType=Bool
    rotationMode=BMM_Additive
    rotationSpace=BCS_BoneSpace
    translationMode=BMM_Additive
    translationSpace=BCS_BoneSpace
    alphaBoolBlend {blendInTime=0.2, blendOption=Linear, blendOutTime=0.2}
    boneToModify {boneName=pack}
    translation {y=-2, z=-2}
   }
   Save cached pose 'anim dynamics'(Pose=Component To Local(ComponentPose=@n28)) {cacheName=anim dynamics}
  ]}
 EventGraph {nodes [
   n1 = Reactivity() {is=title K2Node_CustomEvent}
   > Set Shaking(Shaking:bool=true) -> (Output_Get:bool)
   > Delay(Duration=3.5)
   > Set ElimReactive(ElimReactive:bool=true) -> (Output_Get:bool)
   > Delay(Duration=0.7)
   > Set Shaking(Shaking=false)
   > Delay(Duration=4)
   > Set ElimReactive(ElimReactive=false)
   Event BlueprintUpdateAnimation()
   > n2 = Sequence!(execute) -> (then_0, then_1, then_2:Exec)
   > n3 = RandomRotatorFunction!(
    execute
    GameTimeInSeconds:double=GetGameTimeInSeconds()
    PerlinScale:double=1
    Frequency:Rotator="10,20,20"
    Amplitude:Rotator="5,11,5"
    Offset:Rotator="0, 0, 0"
    TimeOffset:double=0
   ) -> (then, RandomRotator:Rotator)
   > Set Doll Shake R(Doll Shake R:Rotator=@n3) -> (Output_Get:Rotator)
   n4 = GetOwnerSkeletalMeshComponent()
   n5 = BreakRotator(InRot=BreakTransform(InTransform=GetDeltaTransformFromRefPose(self=@n4, BoneName=spine_05, BaseName=spine_04)).Rotation)
   n6 = MapRangeClamped(Value=@n5.Yaw, InRangeA=10, InRangeB=-30, OutRangeB=SelectFloat(A=50, B=30, bPickA=Get bIsSkyDiving()))
   Set backpack_rotate(execute=@n2.then_1, backpack_rotate=MakeRotator(Roll=float + float(A=@n6, B=-8)))
   n7 = SelectFloat(A=-10, B=-5, bPickA=Get bIsSkyDiving())
   > Set backpack_translation(backpack_translation=MakeVector(Y=MapRangeClamped(Value=@n5.Yaw, InRangeA=0, InRangeB=-30, OutRangeA=-0.5, OutRangeB=@n7)))
   Event OnCharacterCustomizationLoaded()
   > n8 = Cast To FortPlayerPawn(Object=TryGetFortPawn())
   > n9 = Cast To B_Athena_PartModifier_Backpack_DryEraseToro!(
    execute
    Object=GetModifierForPartType!(self:Fort Player Pawn Object Reference=@n8, PartType:EFortCustomPartType Enum=Backpack) -> (ReturnValue:Custom Character Part Modifier Object Reference)
   ) -> (then, CastFailed, AsB Athena Part Modifier Backpack Dry Erase Toro:B Athena Part Modifier Backpack Dry Erase Toro Object Reference) {is=title K2Node_DynamicCast}
   > n10 = Branch(Condition=Get ReactivityEnabled!(self:B Athena Part Modifier Backpack Dry Erase Toro Object Reference=@n9) -> (ReactivityEnabled:bool))
   > Set test reactivity(test reactivity:bool=true) -> (Output_Get:bool)
   Set test reactivity(execute=@n10.else, test reactivity=false)
   n11 = Event BlueprintProcessAnimInputEvent()
   > Branch(Condition=IsNamed(self=@n11.AnimInputEvent, InEventName=aie.EnemyPlayerEliminated))
   > Reactivity() {is=title K2Node_CallFunction}
   Event BlueprintBeginPlay() {is=title K2Node_Event}
   > Parent: BlueprintBeginPlay() {is=title K2Node_CallParentFunction}
   > Branch(Condition=Get bShouldDisplayFrontendReactivity() -> (bShouldDisplayFrontendReactivity:bool))
   > Delay(Duration=4)
   > Reactivity()
   > Set Timer by Event(Delegate=@n1, Time=15, bLooping=true)
  ]}
]
variables [
 backpack_rotate {type=Rotator}
 backpack_translation {type=Vector}
 Doll Shake R {type=Rotator}
 ElimReactive {type=boolean}
 Shaking {type=boolean}
 test reactivity {type=boolean}
]
